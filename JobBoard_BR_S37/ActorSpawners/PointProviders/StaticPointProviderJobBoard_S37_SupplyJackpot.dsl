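; Blueprint point provider for the Supply Jackpot in the Season 37 BR Job Board, judging by its name, holding editor preview meshes.
; On BeginPlay it calls the parent, then hides the PreVis meshes in game, propagating to children, so they remain visible only in the editor.
; The construction script is empty.
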(event EventBeginPlay
  ;[0 0 0] We want this visible in editor when 'G' is pressed, but not visible in game
  (|CallParentFunction)
  (Development|SetHiddeninGame (Variables|Default|GetPreVisMeshes) true true))

(fn ConstructionScript)
